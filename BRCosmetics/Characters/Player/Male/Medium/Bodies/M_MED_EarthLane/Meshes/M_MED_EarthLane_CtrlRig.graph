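# Control Rig for the EarthLane male medium body that procedurally drives its jacket and collar bones after a shared DeformRig function.
# Spherical pose readers and remapped leg angles produce alphas that weight parent constraints and transform offsets on the jacket.
# A low-pelvis alpha fades most leg following and blends in a separate pose, suppressed when DBNO-carried.
# Center back bones average both sides, Movement Alpha adds offsets, and collar bones follow neck, spine and head pose.
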ControlRigBlueprint /BRCosmetics/Characters/Player/Male/Medium/Bodies/M_MED_EarthLane/Meshes/M_MED_EarthLane_CtrlRig
graphs [
 RigVMFunctionLibrary {links [], nodes []}
 RigVMModel {nodes [
   RigUnit_BeginExecution = &RigUnit_BeginExecution()
   FN_DeformRig_FNC = RigVMFunctionReferenceNode(
    DeformRig_ Curve="(deformRig_l_upperarm_up_alpha_72_#1=False,deformRig_r_upperarm_up_alpha_73_#2=False,deformRig_l_upperarm_fr_alpha_74_#3=False,deformRig_r_upperarm_fr_alpha_75_CB51B8 ...(truncated)"
    DeformRig_Bool="(upperarm_1_#4=True,elbow_3_#5=True,wrist_7_#6=True,groin_8_#7=True,glute_17_#8=True,knee_9_40C408324CA261 ...(truncated)"
    DeformRig_Debug=(upperarm_debug_18_#4=False,wrist_debug_20_#6=False,lat_debug_26_#9=False,groinGlute_debug_27_#7=False)
    DeformRig_Value="(upperarm_up_18_#10=(Rotation=(X=0.0,Y=-0.0,Z=0.0,W=1.0),Translation=(X=0.0,Y=0.0,Z=8.0),Scale3D=(X=0.70,Y=1.0,Z=1.0)),upperarm_fwd_fr_19_#11=(Rotation=(X=0.0,Y=-0.0,Z=0.0,W=1.0),Translation=(X=0.0, ...(truncated)"
    ExecuteContext=@RigUnit_BeginExecution.ExecutePin
   )
   > n1 = &RigUnit_SphericalPoseReader(
    ActiveRegionScaleFactors=(PositiveWidth=2,NegativeWidth=2,PositiveHeight=2,NegativeHeight=1)
    ActiveRegionSize=0.1
    Debug=(bDrawDebug=true,bDraw2D=False,bDrawLocalAxes=False,DebugScale=25,DebugSegments=20,DebugThickness=0.25)
    DriverAxis=(X=-1,Y=0,Z=0)
    DriverItem=(Type=Bone,Name="thigh_l")
    FalloffRegionScaleFactors=(PositiveWidth=2,NegativeWidth=2,PositiveHeight=4,NegativeHeight=1)
    FalloffSize=0.3
    FlipHeightScaling=false
    FlipWidthScaling=false
    OptionalParentItem=(Type=Bone,Name="None")
    RotationOffset=(X=0,Y=90,Z=80)
   )
   > n2 = &RigUnit_SphericalPoseReader(
    ActiveRegionScaleFactors=(PositiveWidth=2,NegativeWidth=2,PositiveHeight=2,NegativeHeight=1)
    ActiveRegionSize=0.1
    Debug=(bDrawDebug=true,bDraw2D=False,bDrawLocalAxes=False,DebugScale=25,DebugSegments=20,DebugThickness=0.25)
    DriverAxis=(X=1,Y=0,Z=0)
    DriverItem=(Type=Bone,Name="thigh_r")
    FalloffRegionScaleFactors=(PositiveWidth=2,NegativeWidth=2,PositiveHeight=4,NegativeHeight=1)
    FalloffSize=0.3
    FlipHeightScaling=false
    FlipWidthScaling=false
    OptionalParentItem=(Type=Bone,Name="None")
    RotationOffset=(X=180,Y=90,Z=80)
   )
   > n3 = &RigUnit_SphericalPoseReader(
    ActiveRegionScaleFactors=(PositiveWidth=4,NegativeWidth=3,PositiveHeight=2.5,NegativeHeight=2)
    ActiveRegionSize=0.1
    Debug=(bDrawDebug=false,bDraw2D=False,bDrawLocalAxes=False,DebugScale=25,DebugSegments=20,DebugThickness=0.25)
    DriverAxis=(X=-1,Y=0,Z=0)
    DriverItem=(Type=Bone,Name="thigh_l")
    FalloffRegionScaleFactors=(PositiveWidth=1,NegativeWidth=4,PositiveHeight=1,NegativeHeight=0.5)
    FalloffSize=0.4
    FlipHeightScaling=False
    FlipWidthScaling=False
    OptionalParentItem=(Type=Bone,Name="None")
    RotationOffset=(X=100,Y=0,Z=-20)
   )
   > n4 = &RigUnit_SphericalPoseReader(
    ActiveRegionScaleFactors=(PositiveWidth=4,NegativeWidth=3,PositiveHeight=2.5,NegativeHeight=2)
    ActiveRegionSize=0.1
    Debug=(bDrawDebug=false,bDraw2D=False,bDrawLocalAxes=False,DebugScale=25,DebugSegments=20,DebugThickness=0.25)
    DriverAxis=(X=1,Y=0,Z=0)
    DriverItem=(Type=Bone,Name="thigh_r")
    FalloffRegionScaleFactors=(PositiveWidth=1,NegativeWidth=4,PositiveHeight=1,NegativeHeight=0.5)
    FalloffSize=0.4
    FlipHeightScaling=False
    FlipWidthScaling=true
    OptionalParentItem=(Type=Bone,Name="None")
    RotationOffset=(X=-80,Y=0,Z=-20)
   )
   > n5 = &RigUnit_SphericalPoseReader(
    ActiveRegionScaleFactors=(PositiveWidth=3,NegativeWidth=2,PositiveHeight=2,NegativeHeight=2)
    ActiveRegionSize=0.1
    Debug=(bDrawDebug=false,bDraw2D=False,bDrawLocalAxes=False,DebugScale=25,DebugSegments=20,DebugThickness=0.25)
    DriverAxis=(X=-1,Y=0,Z=0)
    DriverItem=(Type=Bone,Name="thigh_l")
    FalloffRegionScaleFactors=(PositiveWidth=2,NegativeWidth=2,PositiveHeight=2,NegativeHeight=2)
    FalloffSize=0.25
    FlipHeightScaling=false
    FlipWidthScaling=false
    OptionalParentItem=(Type=Bone,Name="None")
    RotationOffset=(X=180,Y=-40,Z=0)
   )
   > n6 = &RigUnit_SphericalPoseReader(
    ActiveRegionScaleFactors=(PositiveWidth=3,NegativeWidth=2,PositiveHeight=2,NegativeHeight=2)
    ActiveRegionSize=0.1
    Debug=(bDrawDebug=false,bDraw2D=False,bDrawLocalAxes=False,DebugScale=25,DebugSegments=20,DebugThickness=0.25)
    DriverAxis=(X=1,Y=0,Z=0)
    DriverItem=(Type=Bone,Name="thigh_r")
    FalloffRegionScaleFactors=(PositiveWidth=2,NegativeWidth=2,PositiveHeight=2,NegativeHeight=2)
    FalloffSize=0.25
    FlipHeightScaling=False
    FlipWidthScaling=true
    OptionalParentItem=(Type=Bone,Name="None")
    RotationOffset=(X=0,Y=-40,Z=0)
   )
   RigVMFunction_MathQuaternionToEuler = &RigVMFunction_MathQuaternionToEuler(
    RotationOrder=ZYX
    Value=&RigUnit_GetRelativeTransformForItem(Child=(Type=Bone,Name="calf_l"), Parent=(Type=Bone,Name="thigh_l"), bChildInitial=False, bParentInitial=False).RelativeTransform.Rotation
   )
   Remap = &RigVMFunction_MathDoubleRemap(SourceMaximum=110, SourceMinimum=25, TargetMaximum=1, TargetMinimum=0, Value=@RigVMFunction_MathQuaternionToEuler.Result.Z)
   > RigVMVariableNode(Value=@Remap.Result) {variable=L calfBend Alpha}
   n7 = &RigUnit_GetRelativeTransformForItem(Child=(Type=Bone,Name="calf_r"), Parent=(Type=Bone,Name="thigh_r"), bChildInitial=False, bParentInitial=False)
   > RigVMVariableNode(Value=&RigVMFunction_MathDoubleRemap(
     SourceMaximum=110
     SourceMinimum=25
     TargetMaximum=1
     TargetMinimum=0
     Value=&RigVMFunction_MathQuaternionToEuler(RotationOrder=ZYX, Value=@n7.RelativeTransform.Rotation).Result.Z
    ).Result) {variable=R calfBend Alpha}
   n8 = &RigUnit_GetRelativeTransformForItem(Child=(Type=Bone,Name="thigh_l"), Parent=(Type=Bone,Name="pelvis"), bChildInitial=False, bParentInitial=False)
   > RigVMVariableNode(Value=&RigVMFunction_MathDoubleRemap(
     SourceMaximum=-90
     SourceMinimum=-50
     TargetMaximum=1
     TargetMinimum=0
     Value=&RigVMFunction_MathQuaternionToEuler(RotationOrder=ZYX, Value=@n8.RelativeTransform.Rotation).Result.Z
    ).Result) {variable=L thighVeryForward}
   Absolute = &RigVMFunction_MathDoubleAbs(Value=&RigVMFunction_MathQuaternionToEuler(
     RotationOrder=ZYX
     Value=&RigUnit_GetRelativeTransformForItem(Child=(Type=Bone,Name="thigh_r"), Parent=(Type=Bone,Name="pelvis"), bChildInitial=False, bParentInitial=False).RelativeTransform.Rotation
    ).Result.Z)
   > RigVMVariableNode(Value=&RigVMFunction_MathDoubleRemap(SourceMaximum=90, SourceMinimum=130, TargetMaximum=1, TargetMinimum=0, Value=@Absolute.Result).Result) {variable=R thighVeryForward}
   n9 = &RigUnit_GetRelativeTransformForItem(Child=(Type=Bone,Name="spine_01"), Parent=(Type=Bone,Name="root"), bChildInitial=False, bParentInitial=false)
   n10 = &RigVMFunction_MathDoubleAbs(Value=&RigVMFunction_MathQuaternionToEuler(RotationOrder=XZY, Value=@n9.RelativeTransform.Rotation).Result.Z)
   n11 = &RigVMFunction_MathDoubleRemap(SourceMaximum=30, SourceMinimum=60, TargetMaximum=1, TargetMinimum=0, Value=@n9.RelativeTransform.Translation.Z)
   > RigVMVariableNode(Value=&RigVMFunction_MathDoubleMul(
     A=&RigVMFunction_MathDoubleRemap(SourceMaximum=60, SourceMinimum=35, TargetMaximum=0, TargetMinimum=1, Value=@n10.Result).Result
     B=@n11.Result
    ).Result) {variable=pelvisLowToGround_translateRotate}
   > RigVMVariableNode(Value=&RigVMFunction_MathDoubleSub(A=1, B=@n11.Result).Result) {variable=pelvisLowToGround_INVERSE}
   > RigVMVariableNode(Value=@n1.OutputParam) {variable=L thighBack Alpha}
   > RigVMVariableNode(Value=@n2.OutputParam) {variable=R thighBack Alpha}
   > RigVMVariableNode(Value=@n3.OutputParam) {variable=L thighForward Alpha}
   > RigVMVariableNode(Value=@n4.OutputParam) {variable=R thighForward Alpha}
   > RigVMVariableNode(Value=@n5.OutputParam) {variable=L thighSide Alpha}
   > RigVMVariableNode(Value=@n6.OutputParam) {variable=R thighSide Alpha}
   n12 = &RigVMFunction_MathDoubleAdd(B=RigVMVariableNode() {variable=R thighSide Alpha}.Value, A=RigVMVariableNode() {variable=R thighForward Alpha}.Value)
   > &RigUnit_ParentConstraint(
    AdvancedSettings=(InterpolationType=Average,RotationOrderForFilter=XZY)
    Child=(Type=Bone,Name="dyn_Jacket_r_01_jnt")
    Filter=(TranslationFilter=(bX=true,bY=true,bZ=true),RotationFilter=(bX=true,bY=false,bZ=true),ScaleFilter=(bX=True,bY=True,bZ=True))
    Parents=((Item=(Type=Bone,Name="thigh_r"),Weight=1))
    Weight=&RigVMFunction_MathDoubleMul(A=RigVMVariableNode() {variable=pelvisLowToGround_INVERSE}.Value, B=&RigVMFunction_MathDoubleMul(B=0.75, A=@n12.Result).Result).Result
   )
   n13 = RigVMVariableNode() {variable=R thighSide Alpha}
   n14 = RigVMVariableNode() {variable=R thighForward Alpha}
   > &RigUnit_ParentConstraint(
    AdvancedSettings=(InterpolationType=Average,RotationOrderForFilter=XZY)
    Child=(Type=Bone,Name="dyn_Jacket_r_03_jnt")
    Filter=(TranslationFilter=(bX=false,bY=false,bZ=false),RotationFilter=(bX=false,bY=true,bZ=true),ScaleFilter=(bX=True,bY=True,bZ=True))
    Parents=((Item=(Type=Bone,Name="thigh_r"),Weight=0.5),(Item=(Type=Bone,Name="calf_r"),Weight=0.5))
    Weight=&RigVMFunction_MathDoubleMul(
     A=RigVMVariableNode() {variable=pelvisLowToGround_INVERSE}.Value
     B=&RigVMFunction_MathDoubleSub(A=@n13.Value, B=@n14.Value).Result
    ).Result
   )
   > &RigUnit_ParentConstraint(
    AdvancedSettings=(InterpolationType=Average,RotationOrderForFilter=XZY)
    Child=(Type=Bone,Name="dyn_Jacket_bk_r_02_jnt")
    Filter=(TranslationFilter=(bX=false,bY=false,bZ=false),RotationFilter=(bX=true,bY=true,bZ=true),ScaleFilter=(bX=True,bY=True,bZ=True))
    Parents=((Item=(Type=Bone,Name="thigh_r"),Weight=0.8),(Item=(Type=Bone,Name="calf_r"),Weight=0.2))
    Weight=&RigVMFunction_MathDoubleSub(A=0.3, B=RigVMVariableNode() {variable=R thighForward Alpha}.Value).Result
   )
   Add = &RigVMFunction_MathDoubleAdd(A=RigVMVariableNode() {variable=R calfBend Alpha}.Value, B=RigVMVariableNode() {variable=R thighBack Alpha}.Value)
   n15 = &RigVMFunction_MathDoubleSub(A=1, B=RigVMVariableNode() {variable=R thighForward Alpha}.Value)
   > &RigUnit_ParentConstraint(
    AdvancedSettings=(InterpolationType=Average,RotationOrderForFilter=XZY)
    Child=(Type=Bone,Name="dyn_Jacket_bk_r_03_jnt")
    Filter=(TranslationFilter=(bX=false,bY=false,bZ=false),RotationFilter=(bX=True,bY=True,bZ=True),ScaleFilter=(bX=True,bY=True,bZ=True))
    Parents=((Item=(Type=Bone,Name="calf_r"),Weight=0.65),(Item=(Type=Bone,Name="thigh_r"),Weight=0.35))
    Weight=&RigVMFunction_MathDoubleMul(
     A=RigVMVariableNode() {variable=pelvisLowToGround_INVERSE}.Value
     B=&RigVMFunction_MathDoubleMul(A=@Add.Result, B=@n15.Result).Result
    ).Result
   )
   n16 = &RigVMFunction_MathDoubleAdd(B=RigVMVariableNode() {variable=R thighBack Alpha}.Value, A=RigVMVariableNode() {variable=R calfBend Alpha}.Value)
   n17 = &RigVMFunction_MathDoubleMul(B=0.2, A=RigVMVariableNode() {variable=R thighForward Alpha}.Value)
   > &RigUnit_ModifyTransforms(
    ItemToModify=((Item=(Type=Bone,Name="dyn_Jacket_bk_r_01_jnt"),Transform=(Rotation=(X=-0.0027,Y=-0.0435,Z=-0.061,W=0.9972),Translation=(X=0,Y=0,Z=0),Scale3D=(X=1,Y=1,Z=1))))
    WeightMinimum=0
    Weight=&RigVMFunction_MathDoubleMul(
     B=&RigVMFunction_MathDoubleSub(A=@n16.Result, B=@n17.Result).Result
     A=RigVMVariableNode() {variable=pelvisLowToGround_INVERSE}.Value
    ).Result
   )
   n18 = &RigVMFunction_MathDoubleSub(A=1, B=RigVMVariableNode() {variable=L thighSide Alpha}.Value)
   n19 = RigVMVariableNode() {variable=L calfBend Alpha}
   n20 = RigVMVariableNode() {variable=L thighForward Alpha}
   n21 = &RigVMFunction_MathDoubleMul(
    B=&RigVMFunction_MathDoubleSub(
     A=&RigVMFunction_MathDoubleMul(A=@n18.Result, B=@n19.Value).Result
     B=&RigVMFunction_MathDoubleMul(B=0.7, A=@n20.Value).Result
    ).Result
    A=RigVMVariableNode() {variable=pelvisLowToGround_INVERSE}.Value
   )
   > &RigUnit_ParentConstraint(
    AdvancedSettings=(InterpolationType=Average,RotationOrderForFilter=XZY)
    Child=(Type=Bone,Name="dyn_Jacket_r_02_jnt")
    Filter=(TranslationFilter=(bX=false,bY=false,bZ=false),RotationFilter=(bX=True,bY=True,bZ=True),ScaleFilter=(bX=True,bY=True,bZ=True))
    Parents=((Item=(Type=Bone,Name="calf_l"),Weight=1))
    Weight=@n21.Result
   )
   > &RigUnit_ParentConstraint(
    AdvancedSettings=(InterpolationType=Average,RotationOrderForFilter=XZY)
    Child=(Type=Bone,Name="dyn_Jacket_bk_r_02_jnt")
    Filter=(TranslationFilter=(bX=false,bY=false,bZ=false),RotationFilter=(bX=True,bY=True,bZ=True),ScaleFilter=(bX=True,bY=True,bZ=True))
    Parents=((Item=(Type=Bone,Name="calf_l"),Weight=1))
    Weight=@n21.Result
   )
   > &RigUnit_ModifyTransforms(
    ItemToModify=((Item=(Type=Bone,Name="dyn_Jacket_bk_r_01_jnt"),Transform=(Rotation=(X=0,Y=0,Z=-0.0262,W=0.9997),Translation=(X=0,Y=0,Z=0),Scale3D=(X=1,Y=1,Z=1))))
    WeightMinimum=0
    Weight=@Add.Result
   )
   > &RigUnit_ModifyTransforms(
    ItemToModify=((Item=(Type=Bone,Name="dyn_Jacket_r_01_jnt"),Transform=(Rotation=(X=0,Y=0,Z=-0.0175,W=0.9998),Translation=(X=2,Y=-3,Z=0),Scale3D=(X=1,Y=1,Z=1))))
    WeightMinimum=0
    Weight=RigVMVariableNode() {variable=R thighVeryForward}.Value
   )
   n22 = &RigVMFunction_MathDoubleAdd(B=RigVMVariableNode() {variable=L thighSide Alpha}.Value, A=RigVMVariableNode() {variable=L thighForward Alpha}.Value)
   > &RigUnit_ParentConstraint(
    AdvancedSettings=(InterpolationType=Average,RotationOrderForFilter=XZY)
    Child=(Type=Bone,Name="dyn_Jacket_l_01_jnt")
    Filter=(TranslationFilter=(bX=true,bY=true,bZ=true),RotationFilter=(bX=true,bY=false,bZ=true),ScaleFilter=(bX=True,bY=True,bZ=True))
    Parents=((Item=(Type=Bone,Name="thigh_l"),Weight=1))
    Weight=&RigVMFunction_MathDoubleMul(A=RigVMVariableNode() {variable=pelvisLowToGround_INVERSE}.Value, B=&RigVMFunction_MathDoubleMul(B=0.75, A=@n22.Result).Result).Result
   )
   n23 = RigVMVariableNode() {variable=L thighSide Alpha}
   n24 = RigVMVariableNode() {variable=L thighForward Alpha}
   > &RigUnit_ParentConstraint(
    AdvancedSettings=(InterpolationType=Average,RotationOrderForFilter=XZY)
    Child=(Type=Bone,Name="dyn_Jacket_l_03_jnt")
    Filter=(TranslationFilter=(bX=false,bY=false,bZ=false),RotationFilter=(bX=false,bY=true,bZ=true),ScaleFilter=(bX=True,bY=True,bZ=True))
    Parents=((Item=(Type=Bone,Name="thigh_l"),Weight=0.5),(Item=(Type=Bone,Name="calf_l"),Weight=0.5))
    Weight=&RigVMFunction_MathDoubleMul(
     A=RigVMVariableNode() {variable=pelvisLowToGround_INVERSE}.Value
     B=&RigVMFunction_MathDoubleSub(A=@n23.Value, B=@n24.Value).Result
    ).Result
   )
   > &RigUnit_ParentConstraint(
    AdvancedSettings=(InterpolationType=Average,RotationOrderForFilter=XZY)
    Child=(Type=Bone,Name="dyn_Jacket_bk_l_02_jnt")
    Filter=(TranslationFilter=(bX=false,bY=false,bZ=false),RotationFilter=(bX=True,bY=True,bZ=True),ScaleFilter=(bX=True,bY=True,bZ=True))
    Parents=((Item=(Type=Bone,Name="thigh_l"),Weight=0.8),(Item=(Type=Bone,Name="calf_l"),Weight=0.2))
    Weight=&RigVMFunction_MathDoubleSub(A=0.3, B=RigVMVariableNode() {variable=L thighForward Alpha}.Value).Result
   )
   n25 = &RigVMFunction_MathDoubleAdd(B=RigVMVariableNode() {variable=L thighBack Alpha}.Value, A=RigVMVariableNode() {variable=L calfBend Alpha}.Value)
   n26 = &RigVMFunction_MathDoubleSub(A=1, B=RigVMVariableNode() {variable=L thighForward Alpha}.Value)
   > &RigUnit_ParentConstraint(
    AdvancedSettings=(InterpolationType=Average,RotationOrderForFilter=XZY)
    Child=(Type=Bone,Name="dyn_Jacket_bk_l_03_jnt")
    Filter=(TranslationFilter=(bX=false,bY=false,bZ=false),RotationFilter=(bX=True,bY=True,bZ=True),ScaleFilter=(bX=True,bY=True,bZ=True))
    Parents=((Item=(Type=Bone,Name="calf_l"),Weight=0.65),(Item=(Type=Bone,Name="thigh_l"),Weight=0.35))
    Weight=&RigVMFunction_MathDoubleMul(
     A=RigVMVariableNode() {variable=pelvisLowToGround_INVERSE}.Value
     B=&RigVMFunction_MathDoubleMul(A=@n25.Result, B=@n26.Result).Result
    ).Result
   )
   n27 = &RigVMFunction_MathDoubleAdd(B=RigVMVariableNode() {variable=L thighBack Alpha}.Value, A=RigVMVariableNode() {variable=L calfBend Alpha}.Value)
   n28 = &RigVMFunction_MathDoubleMul(B=0.2, A=RigVMVariableNode() {variable=L thighForward Alpha}.Value)
   > &RigUnit_ModifyTransforms(
    ItemToModify=((Item=(Type=Bone,Name="dyn_Jacket_bk_l_01_jnt"),Transform=(Rotation=(X=0.0027,Y=0.0435,Z=-0.061,W=0.9972),Translation=(X=0,Y=0,Z=0),Scale3D=(X=1,Y=1,Z=1))))
    WeightMinimum=0
    Weight=&RigVMFunction_MathDoubleMul(
     B=&RigVMFunction_MathDoubleSub(A=@n27.Result, B=@n28.Result).Result
     A=RigVMVariableNode() {variable=pelvisLowToGround_INVERSE}.Value
    ).Result
   )
   Subtract = &RigVMFunction_MathDoubleSub(A=1, B=RigVMVariableNode() {variable=R thighSide Alpha}.Value)
   Multiply = &RigVMFunction_MathDoubleMul(A=@Subtract.Result, B=RigVMVariableNode() {variable=R calfBend Alpha}.Value)
   n29 = RigVMVariableNode() {variable=R thighForward Alpha}
   n30 = &RigVMFunction_MathDoubleMul(
    B=&RigVMFunction_MathDoubleSub(A=@Multiply.Result, B=&RigVMFunction_MathDoubleMul(B=0.7, A=@n29.Value).Result).Result
    A=RigVMVariableNode() {variable=pelvisLowToGround_INVERSE}.Value
   )
   > &RigUnit_ParentConstraint(
    AdvancedSettings=(InterpolationType=Average,RotationOrderForFilter=XZY)
    Child=(Type=Bone,Name="dyn_Jacket_l_02_jnt")
    Filter=(TranslationFilter=(bX=false,bY=false,bZ=false),RotationFilter=(bX=True,bY=True,bZ=True),ScaleFilter=(bX=True,bY=True,bZ=True))
    Parents=((Item=(Type=Bone,Name="calf_r"),Weight=1))
    Weight=@n30.Result
   )
   > &RigUnit_ParentConstraint(
    AdvancedSettings=(InterpolationType=Average,RotationOrderForFilter=XZY)
    Child=(Type=Bone,Name="dyn_Jacket_bk_l_02_jnt")
    Filter=(TranslationFilter=(bX=false,bY=false,bZ=false),RotationFilter=(bX=True,bY=True,bZ=True),ScaleFilter=(bX=True,bY=True,bZ=True))
    Parents=((Item=(Type=Bone,Name="calf_r"),Weight=1))
    Weight=@n30.Result
   )
   > &RigUnit_ModifyTransforms(
    ItemToModify=((Item=(Type=Bone,Name="dyn_Jacket_bk_l_01_jnt"),Transform=(Rotation=(X=0,Y=0,Z=-0.0262,W=0.9997),Translation=(X=0,Y=0,Z=0),Scale3D=(X=1,Y=1,Z=1))))
    WeightMinimum=0
    Weight=@n25.Result
   )
   > &RigUnit_ModifyTransforms(
    ItemToModify=((Item=(Type=Bone,Name="dyn_Jacket_l_01_jnt"),Transform=(Rotation=(X=0,Y=0,Z=-0.0175,W=0.9998),Translation=(X=2,Y=-3,Z=0),Scale3D=(X=1,Y=1,Z=1))))
    WeightMinimum=0
    Weight=RigVMVariableNode() {variable=L thighVeryForward}.Value
   )
   Cast = &RigVMFunction_MathBoolToFloat(Value=RigVMVariableNode() {variable=is_DBNO_Carried}.Value)
   n31 = &RigVMFunction_MathDoubleMul(
    A=&RigVMFunction_MathDoubleSub(A=1, B=@Cast.Result).Result
    B=RigVMVariableNode() {variable=pelvisLowToGround_translateRotate}.Value
   )
   > &RigUnit_ModifyTransforms(
    ItemToModify=((Item=(Type=Bone,Name="dyn_Jacket_l_01_jnt"),Transform=(Rotation=(X=0,Y=0,Z=-0.6088,W=0.7934),Translation=(X=0,Y=0,Z=0),Scale3D=(X=1,Y=1,Z=1))))
    WeightMinimum=0
    Weight=@n31.Result
   )
   > &RigUnit_ModifyTransforms(
    ItemToModify=((Item=(Type=Bone,Name="dyn_Jacket_l_02_jnt"),Transform=(Rotation=(X=0,Y=0,Z=-0.0872,W=0.9962),Translation=(X=0,Y=0,Z=0),Scale3D=(X=1,Y=1,Z=1))))
    WeightMinimum=0
    Weight=@n31.Result
   )
   > &RigUnit_ModifyTransforms(
    ItemToModify=((Item=(Type=Bone,Name="dyn_Jacket_l_03_jnt"),Transform=(Rotation=(X=0,Y=0,Z=0,W=1),Translation=(X=0,Y=0,Z=0),Scale3D=(X=1,Y=1,Z=1))))
    WeightMinimum=0
    Weight=@n31.Result
   )
   > &RigUnit_ModifyTransforms(
    ItemToModify=((Item=(Type=Bone,Name="dyn_Jacket_r_01_jnt"),Transform=(Rotation=(X=0,Y=0,Z=-0.6088,W=0.7934),Translation=(X=0,Y=0,Z=0),Scale3D=(X=1,Y=1,Z=1))))
    WeightMinimum=0
    Weight=@n31.Result
   )
   > &RigUnit_ModifyTransforms(
    ItemToModify=((Item=(Type=Bone,Name="dyn_Jacket_r_02_jnt"),Transform=(Rotation=(X=0,Y=0,Z=-0.0872,W=0.9962),Translation=(X=0,Y=0,Z=0),Scale3D=(X=1,Y=1,Z=1))))
    WeightMinimum=0
    Weight=@n31.Result
   )
   > &RigUnit_ModifyTransforms(
    ItemToModify=((Item=(Type=Bone,Name="dyn_Jacket_r_03_jnt"),Transform=(Rotation=(X=0,Y=0,Z=0,W=1),Translation=(X=0,Y=0,Z=0),Scale3D=(X=1,Y=1,Z=1))))
    WeightMinimum=0
    Weight=@n31.Result
   )
   > &RigUnit_ModifyTransforms(
    ItemToModify=((Item=(Type=Bone,Name="dyn_Jacket_bk_l_01_jnt"),Transform=(Rotation=(X=-0.03,Y=-0.0316,Z=-0.6877,W=0.7247),Translation=(X=-2,Y=-2,Z=0),Scale3D=(X=1,Y=1,Z=1))))
    WeightMinimum=0
    Weight=@n31.Result
   )
   > &RigUnit_ModifyTransforms(
    ItemToModify=((Item=(Type=Bone,Name="dyn_Jacket_bk_l_02_jnt"),Transform=(Rotation=(X=0,Y=0,Z=-0.0872,W=0.9962),Translation=(X=1,Y=-2,Z=0),Scale3D=(X=1,Y=1,Z=1))))
    WeightMinimum=0
    Weight=@n31.Result
   )
   > &RigUnit_ModifyTransforms(
    ItemToModify=((Item=(Type=Bone,Name="dyn_Jacket_bk_l_03_jnt"),Transform=(Rotation=(X=0,Y=0,Z=0,W=1),Translation=(X=0,Y=0,Z=0),Scale3D=(X=1,Y=1,Z=1))))
    WeightMinimum=0
    Weight=@n31.Result
   )
   > &RigUnit_ModifyTransforms(
    ItemToModify=((Item=(Type=Bone,Name="dyn_Jacket_bk_r_01_jnt"),Transform=(Rotation=(X=0.03,Y=0.0316,Z=-0.6877,W=0.7247),Translation=(X=-2,Y=-2,Z=0),Scale3D=(X=1,Y=1,Z=1))))
    WeightMinimum=0
    Weight=@n31.Result
   )
   > &RigUnit_ModifyTransforms(
    ItemToModify=((Item=(Type=Bone,Name="dyn_Jacket_bk_r_02_jnt"),Transform=(Rotation=(X=0,Y=0,Z=-0.0872,W=0.9962),Translation=(X=1,Y=-2,Z=0),Scale3D=(X=1,Y=1,Z=1))))
    WeightMinimum=0
    Weight=@n31.Result
   )
   > &RigUnit_ModifyTransforms(
    ItemToModify=((Item=(Type=Bone,Name="dyn_Jacket_bk_r_03_jnt"),Transform=(Rotation=(X=0,Y=0,Z=0,W=1),Translation=(X=0,Y=0,Z=0),Scale3D=(X=1,Y=1,Z=1))))
    WeightMinimum=0
    Weight=@n31.Result
   )
   > &RigUnit_ParentConstraint(
    AdvancedSettings=(InterpolationType=Average,RotationOrderForFilter=XZY)
    Child=(Type=Bone,Name="dyn_Jacket_bk_c_01_jnt")
    Filter=(TranslationFilter=(bX=true,bY=true,bZ=true),RotationFilter=(bX=True,bY=True,bZ=True),ScaleFilter=(bX=True,bY=True,bZ=True))
    Parents=((Item=(Type=Bone,Name="dyn_Jacket_bk_l_01_jnt"),Weight=0.5),(Item=(Type=Bone,Name="dyn_Jacket_bk_r_01_jnt"),Weight=0.5))
    Weight=1
   )
   > &RigUnit_ParentConstraint(
    AdvancedSettings=(InterpolationType=Average,RotationOrderForFilter=XZY)
    Child=(Type=Bone,Name="dyn_Jacket_bk_c_02_jnt")
    Filter=(TranslationFilter=(bX=true,bY=true,bZ=true),RotationFilter=(bX=True,bY=True,bZ=True),ScaleFilter=(bX=True,bY=True,bZ=True))
    Parents=((Item=(Type=Bone,Name="dyn_Jacket_bk_l_02_jnt"),Weight=0.5),(Item=(Type=Bone,Name="dyn_Jacket_bk_r_02_jnt"),Weight=0.5))
    Weight=1
    bMaintainOffset=true
   )
   > &RigUnit_ParentConstraint(
    AdvancedSettings=(InterpolationType=Average,RotationOrderForFilter=XZY)
    Child=(Type=Bone,Name="dyn_Jacket_bk_c_03_jnt")
    Parents=((Item=(Type=Bone,Name="dyn_Jacket_bk_l_03_jnt"),Weight=0.5),(Item=(Type=Bone,Name="dyn_Jacket_bk_r_03_jnt"),Weight=0.5))
    Weight=1
   )
   > &RigUnit_ParentConstraint(
    AdvancedSettings=(InterpolationType=Average,RotationOrderForFilter=XZY)
    Child=(Type=Bone,Name="dyn_Jacket_bk_c_04_jnt")
    Parents=((Item=(Type=Bone,Name="dyn_Jacket_bk_l_04_jnt"),Weight=0.5),(Item=(Type=Bone,Name="dyn_Jacket_bk_r_04_jnt"),Weight=0.5))
    Weight=1
   )
   > &RigUnit_ParentConstraint(
    AdvancedSettings=(InterpolationType=Average,RotationOrderForFilter=XZY)
    Child=(Type=Bone,Name="dyn_Jacket_bk_c_05_jnt")
    Parents=((Item=(Type=Bone,Name="dyn_Jacket_bk_l_05_jnt"),Weight=0.5),(Item=(Type=Bone,Name="dyn_Jacket_bk_r_05_jnt"),Weight=0.5))
    Weight=1
   )
   n32 = &RigVMFunction_MathBoolToFloat(Value=RigVMVariableNode() {variable=is_DBNO_Carried}.Value)
   > &RigUnit_ModifyTransforms(
    ItemToModify=((Item=(Type=Bone,Name="dyn_Jacket_bk_c_01_jnt"),Transform=(Rotation=(X=0,Y=0,Z=0,W=1),Translation=(X=-5,Y=0,Z=0),Scale3D=(X=1,Y=1,Z=1))))
    WeightMinimum=0
    Weight=&RigVMFunction_MathDoubleMul(
     A=&RigVMFunction_MathDoubleSub(A=1, B=@n32.Result).Result
     B=RigVMVariableNode() {variable=pelvisLowToGround_translateRotate}.Value
    ).Result
   )
   n33 = RigVMVariableNode() {variable=L calfBend Alpha}
   n34 = RigVMVariableNode() {variable=L thighBack Alpha}
   n35 = RigVMVariableNode() {variable=R thighBack Alpha}
   n36 = RigVMVariableNode() {variable=R calfBend Alpha}
   Maximum = &RigVMFunction_MathDoubleMax(
    B=&RigVMFunction_MathDoubleSub(A=&RigVMFunction_MathDoubleAdd(A=@n33.Value, B=@n34.Value).Result, B=RigVMVariableNode() {variable=L thighForward Alpha}.Value).Result
    A=&RigVMFunction_MathDoubleSub(A=&RigVMFunction_MathDoubleAdd(B=@n35.Value, A=@n36.Value).Result, B=RigVMVariableNode() {variable=R thighForward Alpha}.Value).Result
   )
   > &RigUnit_ModifyTransforms(
    ItemToModify=((Item=(Type=Bone,Name="dyn_Jacket_bk_c_01_jnt"),Transform=(Rotation=(X=0,Y=0,Z=-0.0523,W=0.9986),Translation=(X=0,Y=0,Z=0),Scale3D=(X=1,Y=1,Z=1))))
    WeightMinimum=0
    Weight=@Maximum.Result
   )
   > &RigUnit_ModifyTransforms(
    ItemToModify=((Item=(Type=Bone,Name="dyn_Jacket_bk_c_02_jnt"),Transform=(Rotation=(X=0,Y=0,Z=-0.0523,W=0.9986),Translation=(X=0,Y=0,Z=0),Scale3D=(X=1,Y=1,Z=1))))
    WeightMinimum=0
    Weight=@Maximum.Result
   )
   n37 = RigVMVariableNode() {variable=Movement Alpha}
   > &RigUnit_ModifyTransforms(
    ItemToModify=((Item=(Type=Bone,Name="dyn_Jacket_l_01_jnt"),Transform=(Rotation=(X=0.0018,Y=0.0697,Z=-0.0261,W=0.9972),Translation=(X=0,Y=0,Z=0),Scale3D=(X=1,Y=1,Z=1))))
    WeightMinimum=0
    Weight=@n37.Value
   )
   > &RigUnit_ModifyTransforms(
    ItemToModify=((Item=(Type=Bone,Name="dyn_Jacket_bk_l_01_jnt"),Transform=(Rotation=(X=0.0033,Y=0.0347,Z=-0.0958,W=0.9948),Translation=(X=0,Y=0,Z=0),Scale3D=(X=1,Y=1,Z=1))))
    WeightMinimum=0
    Weight=@n37.Value
   )
   > &RigUnit_ModifyTransforms(
    ItemToModify=((Item=(Type=Bone,Name="dyn_Jacket_bk_c_01_jnt"),Transform=(Rotation=(X=0,Y=0,Z=-0.0958,W=0.9954),Translation=(X=0,Y=0,Z=0),Scale3D=(X=1,Y=1,Z=1))))
    WeightMinimum=0
    Weight=@n37.Value
   )
   > &RigUnit_ModifyTransforms(
    ItemToModify=((Item=(Type=Bone,Name="dyn_Jacket_r_01_jnt"),Transform=(Rotation=(X=-0.0018,Y=-0.0697,Z=-0.0261,W=0.9972),Translation=(X=0,Y=0,Z=0),Scale3D=(X=1,Y=1,Z=1))))
    WeightMinimum=0
    Weight=@n37.Value
   )
   > &RigUnit_ModifyTransforms(
    ItemToModify=((Item=(Type=Bone,Name="dyn_Jacket_bk_r_01_jnt"),Transform=(Rotation=(X=-0.0033,Y=-0.0347,Z=-0.0958,W=0.9948),Translation=(X=0,Y=0,Z=0),Scale3D=(X=1,Y=1,Z=1))))
    WeightMinimum=0
    Weight=@n37.Value
   )
   > &RigUnit_ParentConstraint(
    AdvancedSettings=(InterpolationType=Average,RotationOrderForFilter=XZY)
    Child=(Type=Bone,Name="collar_bk_c_01_jnt")
    Parents=((Item=(Type=Bone,Name="neck_02"),Weight=0.3),(Item=(Type=Bone,Name="neck_01"),Weight=0.3),(Item=(Type=Bone,Name="spine_05"),Weight=0.4))
    Weight=1
   )
   > &RigUnit_ParentConstraint(
    AdvancedSettings=(InterpolationType=Average,RotationOrderForFilter=XZY)
    Child=(Type=Bone,Name="collar_bk_l_01_jnt")
    Parents=((Item=(Type=Bone,Name="neck_02"),Weight=0.3),(Item=(Type=Bone,Name="neck_01"),Weight=0.3),(Item=(Type=Bone,Name="spine_05"),Weight=0.4))
    Weight=1
   )
   > &RigUnit_ParentConstraint(
    AdvancedSettings=(InterpolationType=Average,RotationOrderForFilter=XZY)
    Child=(Type=Bone,Name="collar_bk_r_01_jnt")
    Parents=((Item=(Type=Bone,Name="neck_02"),Weight=0.3),(Item=(Type=Bone,Name="neck_01"),Weight=0.3),(Item=(Type=Bone,Name="spine_05"),Weight=0.4))
    Weight=1
   )
   > &RigUnit_ParentConstraint(
    AdvancedSettings=(InterpolationType=Average,RotationOrderForFilter=XZY)
    Child=(Type=Bone,Name="collar_ft_l_01_jnt")
    Parents=((Item=(Type=Bone,Name="neck_02"),Weight=0.3),(Item=(Type=Bone,Name="neck_01"),Weight=0.3),(Item=(Type=Bone,Name="spine_05"),Weight=0.4))
    Weight=1
   )
   > &RigUnit_ParentConstraint(
    AdvancedSettings=(InterpolationType=Average,RotationOrderForFilter=XZY)
    Child=(Type=Bone,Name="collar_ft_r_01_jnt")
    Parents=((Item=(Type=Bone,Name="neck_02"),Weight=0.3),(Item=(Type=Bone,Name="neck_01"),Weight=0.3),(Item=(Type=Bone,Name="spine_05"),Weight=0.4))
    Weight=1
   )
   > n38 = &RigUnit_SphericalPoseReader(
    ActiveRegionScaleFactors=(PositiveWidth=1,NegativeWidth=1,PositiveHeight=1,NegativeHeight=1)
    ActiveRegionSize=0.1
    Debug=(bDrawDebug=false,DebugScale=25,DebugSegments=20,DebugThickness=0.25)
    DriverAxis=(X=0,Y=-1,Z=0)
    DriverItem=(Type=Bone,Name="head")
    FalloffRegionScaleFactors=(PositiveWidth=2,NegativeWidth=2,PositiveHeight=2,NegativeHeight=2)
    FalloffSize=0.2
    FlipWidthScaling=false
    OptionalParentItem=(Type=Bone,Name="neck_01")
    RotationOffset=(X=-90,Y=0,Z=-55)
   )
   > &RigUnit_ModifyTransforms(
    ItemToModify=((Item=(Type=Bone,Name="collar_bk_c_01_jnt"),Transform=(Rotation=(X=0,Y=0,Z=0,W=1),Translation=(X=-1,Y=-1,Z=0),Scale3D=(X=1,Y=1,Z=1))))
    ItemToModify.0.Transform.Rotation=&RigVMFunction_MathQuaternionFromEuler(Euler=(X=0,Y=0,Z=-20)).Result
    Weight=@n38.OutputParam
   )
   > &RigUnit_ModifyTransforms(
    ItemToModify=((Item=(Type=Bone,Name="collar_bk_l_01_jnt"),Transform=(Rotation=(X=0,Y=0,Z=0,W=1),Translation=(X=-1,Y=-1,Z=0),Scale3D=(X=1,Y=1,Z=1))))
    ItemToModify.0.Transform.Rotation=&RigVMFunction_MathQuaternionFromEuler(Euler=(X=0,Y=20,Z=-20)).Result
    Weight=@n38.OutputParam
   )
   > &RigUnit_ModifyTransforms(
    ItemToModify=((Item=(Type=Bone,Name="collar_bk_r_01_jnt"),Transform=(Rotation=(X=0,Y=0,Z=0,W=1),Translation=(X=-1,Y=-1,Z=0),Scale3D=(X=1,Y=1,Z=1))))
    ItemToModify.0.Transform.Rotation=&RigVMFunction_MathQuaternionFromEuler(Euler=(X=0,Y=20,Z=-20)).Result
    Weight=@n38.OutputParam
   )
   > n39 = &RigUnit_SphericalPoseReader(
    ActiveRegionScaleFactors=(PositiveWidth=2,NegativeWidth=2,PositiveHeight=2.5,NegativeHeight=0.5)
    ActiveRegionSize=0.1
    Debug=(bDrawDebug=true,DebugScale=25,DebugSegments=20,DebugThickness=0.25)
    DriverAxis=(X=0,Y=-1,Z=0)
    DriverItem=(Type=Bone,Name="head")
    FalloffRegionScaleFactors=(PositiveWidth=1,NegativeWidth=0.5,PositiveHeight=2,NegativeHeight=1)
    FalloffSize=0.3
    OptionalParentItem=(Type=Bone,Name="neck_02")
    RotationOffset=(X=-35,Y=0,Z=-10)
   )
   > &RigUnit_ModifyTransforms(
    ItemToModify=((Item=(Type=Bone,Name="collar_bk_c_01_jnt"),Transform=(Rotation=(X=0,Y=0,Z=0,W=1),Translation=(X=0,Y=0,Z=0),Scale3D=(X=1,Y=1,Z=1))))
    ItemToModify.0.Transform.Rotation=&RigVMFunction_MathQuaternionFromEuler(Euler=(X=0,Y=0,Z=-20)).Result
    Weight=@n39.OutputParam
   )
   > &RigUnit_ModifyTransforms(
    ItemToModify=((Item=(Type=Bone,Name="collar_bk_l_01_jnt"),Transform=(Rotation=(X=0,Y=0,Z=0,W=1),Translation=(X=0,Y=0,Z=0),Scale3D=(X=1,Y=1,Z=1))))
    ItemToModify.0.Transform.Rotation=&RigVMFunction_MathQuaternionFromEuler(Euler=(X=0,Y=40,Z=0)).Result
    Weight=@n39.OutputParam
   )
   > &RigUnit_ModifyTransforms(
    ItemToModify=((Item=(Type=Bone,Name="collar_ft_l_01_jnt"),Transform=(Rotation=(X=0,Y=0,Z=0,W=1),Translation=(X=0,Y=0,Z=0),Scale3D=(X=1,Y=1,Z=1))))
    ItemToModify.0.Transform.Rotation=&RigVMFunction_MathQuaternionFromEuler(Euler=(X=0,Y=40,Z=0)).Result
    Weight=@n39.OutputParam
   )
   > n40 = &RigUnit_SphericalPoseReader(
    ActiveRegionScaleFactors=(PositiveWidth=2,NegativeWidth=2,PositiveHeight=0.5,NegativeHeight=2.5)
    ActiveRegionSize=0.1
    Debug=(bDrawDebug=true,DebugScale=25,DebugSegments=20,DebugThickness=0.25)
    DriverAxis=(X=0,Y=-1,Z=0)
    DriverItem=(Type=Bone,Name="head")
    FalloffRegionScaleFactors=(PositiveWidth=1,NegativeWidth=0.5,PositiveHeight=1,NegativeHeight=2)
    FalloffSize=0.3
    OptionalParentItem=(Type=Bone,Name="neck_02")
    RotationOffset=(X=215,Y=0,Z=-10)
   )
   > &RigUnit_ModifyTransforms(
    ItemToModify=((Item=(Type=Bone,Name="collar_bk_c_01_jnt"),Transform=(Rotation=(X=0,Y=0,Z=0,W=1),Translation=(X=0,Y=0,Z=0),Scale3D=(X=1,Y=1,Z=1))))
    ItemToModify.0.Transform.Rotation=&RigVMFunction_MathQuaternionFromEuler(Euler=(X=0,Y=0,Z=-20)).Result
    Weight=@n40.OutputParam
   )
   > &RigUnit_ModifyTransforms(
    ItemToModify=((Item=(Type=Bone,Name="collar_bk_r_01_jnt"),Transform=(Rotation=(X=0,Y=0,Z=0,W=1),Translation=(X=0,Y=0,Z=0),Scale3D=(X=1,Y=1,Z=1))))
    ItemToModify.0.Transform.Rotation=&RigVMFunction_MathQuaternionFromEuler(Euler=(X=0,Y=40,Z=0)).Result
    Weight=@n40.OutputParam
   )
   > &RigUnit_ModifyTransforms(
    ItemToModify=((Item=(Type=Bone,Name="collar_ft_r_01_jnt"),Transform=(Rotation=(X=0,Y=0,Z=0,W=1),Translation=(X=0,Y=0,Z=0),Scale3D=(X=1,Y=1,Z=1))))
    ItemToModify.0.Transform.Rotation=&RigVMFunction_MathQuaternionFromEuler(Euler=(X=0,Y=40,Z=0)).Result
    Weight=@n40.OutputParam
   )
   &RigUnit_ParentConstraint(
    AdvancedSettings=(InterpolationType=Average,RotationOrderForFilter=XZY)
    Child=(Type=Bone,Name="dyn_Jacket_r_02_jnt")
    ExecutePin=()
    Filter=(TranslationFilter=(bX=false,bY=false,bZ=false),RotationFilter=(bX=false,bY=true,bZ=true),ScaleFilter=(bX=True,bY=True,bZ=True))
    Parents=((Item=(Type=Bone,Name="dyn_Jacket_bk_r_02_jnt"),Weight=1))
    Weight=0.2
   )
   &RigUnit_ParentConstraint(
    AdvancedSettings=(InterpolationType=Average,RotationOrderForFilter=XZY)
    Child=(Type=Bone,Name="dyn_Jacket_l_02_jnt")
    ExecutePin=()
    Filter=(TranslationFilter=(bX=false,bY=false,bZ=false),RotationFilter=(bX=false,bY=True,bZ=True),ScaleFilter=(bX=True,bY=True,bZ=True))
    Parents=((Item=(Type=Bone,Name="dyn_Jacket_bk_l_02_jnt"),Weight=1))
    Weight=0.2
   )
  ]}
]
hierarchy [
 root {loc=<0 0 0>, rot=<0 0 0 1>}
 pelvis {parent=root, loc=<0 -0.7071 93.1191>, rot=<0.0127 0.707 -0.0127 -0.707>}
 spine_01 {parent=pelvis, loc=<6.7494 -0.6388 0>, rot=<0 0 0.034 -0.9994>}
 spine_02 {parent=spine_01, loc=<7.2518 -0.035 0>, rot=<0 0 0.0238 -0.9997>}
 spine_03 {parent=spine_02, loc=<6.0104 0 0>, rot=<0 0 0.0149 -0.9999>}
 spine_04 {parent=spine_03, loc=<8.1024 0 0>, rot=<0 0 -0.0132 -0.9999>}
 spine_05 {parent=spine_04, loc=<10.0146 0 0>, rot=<0 0 -0.0015 -1>}
 attach_cape {parent=spine_05, loc=<9.017 -11.714 0>, rot=<0.0616 -0.7044 -0.0616 -0.7044>}
 neck_01 {parent=spine_05, loc=<11.3138 0.1529 0>, rot=<0 0 -0.2003 -0.9797>}
 neck_02 {parent=neck_01, loc=<4.5627 0 0>, rot=<0 0 0.0504 -0.9987>}
 head {parent=neck_02, loc=<5.3914 0 0>, rot=<0 0 0.115 -0.9934>}
 attach_backpack {parent=spine_05, loc=<-0.123 -13.473 0>, rot=<-0.0093 -0.707 0.0093 -0.707>}
 clavicle_l {parent=spine_05, loc=<4.3575 -0.1466 -2.3223>, rot=<0.0465 -0.7289 -0.009 -0.6829>}
 upperarm_l {parent=clavicle_l, loc=<15.5111 0 0>, rot=<0.0226 -0.3234 -0.0088 -0.9459>}
 lowerarm_l {parent=upperarm_l, loc=<26.6514 0 0>, rot=<0.0391 -0.0099 -0.2453 -0.9686>}
 hand_l {parent=lowerarm_l, loc=<22.3491 0 0>, rot=<-0.7081 -0.0507 0.0975 -0.6975>}
 thumb_01_l {parent=hand_l, loc=<1.3118 -0.3191 -2.459>, rot=<0.6546 -0.4042 0.0664 -0.6354>}
 thumb_02_l {parent=thumb_01_l, loc=<5.9336 -0.0448 0.0401>, rot=<-0.0037 0.0213 0.298 -0.9543>}
 thumb_03_l {parent=thumb_02_l, loc=<2.2223 0.3682 -0.0907>, rot=<0.0118 -0.013 0.0066 -0.9998>}
 index_metacarpal_l {parent=hand_l, loc=<1.6535 0.0183 -2.96>, rot=<0.0158 -0.0124 -0.0173 -0.9996>}
 index_01_l {parent=index_metacarpal_l, loc=<8.3008 0.0357 -0.0524>, rot=<-0.0044 0.0485 0.0924 -0.9945>}
 index_02_l {parent=index_01_l, loc=<3.3587 -0.3004 0.0452>, rot=<-0.002 0.0027 0.3225 -0.9466>}
 index_03_l {parent=index_02_l, loc=<2.5578 -0.123 -0.0252>, rot=<-0.002 0.0067 0.2558 -0.9667>}
 middle_metacarpal_l {parent=hand_l, loc=<2.1856 -0.1725 -0.5984>, rot=<0.006 0.0168 -0.0329 -0.9993>}
 middle_01_l {parent=middle_metacarpal_l, loc=<7.3275 -0.2447 0.0129>, rot=<-0.0208 0.0042 0.0858 -0.9961>}
 middle_02_l {parent=middle_01_l, loc=<4.0123 0.1401 0.1175>, rot=<-0.0218 0.0258 0.3512 -0.9357>}
 middle_03_l {parent=middle_02_l, loc=<2.6077 0.0534 0.1373>, rot=<0.0082 0.0188 0.1708 -0.9851>}
 ring_metacarpal_l {parent=hand_l, loc=<2.2074 -0.3625 1.3638>, rot=<-0.0284 0.0722 -0.0458 -0.9959>}
 ring_01_l {parent=ring_metacarpal_l, loc=<7.0295 -0.6197 -0.1446>, rot=<-0.0489 -0.0252 0.1892 -0.9804>}
 ring_02_l {parent=ring_01_l, loc=<3.6174 0.4647 -0.1027>, rot=<0 0 0.2806 -0.9598>}
 ring_03_l {parent=ring_02_l, loc=<2.4152 0.114 0.0199>, rot=<0 0 0.1189 -0.9929>}
 pinky_metacarpal_l {parent=hand_l, loc=<1.7616 -0.9753 3.3814>, rot=<-0.0397 0.0814 -0.069 -0.9935>}
 pinky_01_l {parent=pinky_metacarpal_l, loc=<7.1709 -0.4831 -0.1094>, rot=<-0.0654 -0.0035 0.2242 -0.9723>}
 pinky_02_l {parent=pinky_01_l, loc=<2.8274 0.0893 -0.1043>, rot=<0.0037 0.0011 0.2854 -0.9584>}
 pinky_03_l {parent=pinky_02_l, loc=<2.2046 0.0307 -0.0078>, rot=<-0.0002 0.0098 0.1445 -0.9895>}
 deform_wrist_dn_base_l {parent=hand_l, loc=<0 -2.2342 0>, rot=<0 0 0 -1>}
 deform_wrist_dn_l {parent=deform_wrist_dn_base_l, loc=<0 0 0>, rot=<0 0 0 1>}
 deform_wrist_up_base_l {parent=hand_l, loc=<0 2.2682 0>, rot=<0 0 0 -1>}
 deform_wrist_up_l {parent=deform_wrist_up_base_l, loc=<0 0 0>, rot=<0 0 0 1>}
 lowerarm_twist_02_l {parent=lowerarm_l, loc=<9.6199 0 0>, rot=<0 0 0 -1>}
 lowerarm_twist_01_l {parent=lowerarm_l, loc=<17.8549 0 0>, rot=<0 0 0 -1>}
 deform_elbow_in_base_l {parent=lowerarm_l, loc=<0.2112 5.259 0.8274>, rot=<0 0 0 -1>}
 deform_elbow_in_l {parent=deform_elbow_in_base_l, loc=<0 0 0>, rot=<0 0 0 1>}
 deform_elbow_out_base_l {parent=lowerarm_l, loc=<0 -4 0>, rot=<0 0 0 -1>}
 deform_elbow_out_l {parent=deform_elbow_out_base_l, loc=<0 0 0>, rot=<0 0 0 1>}
 upperarm_twist_01_l {parent=upperarm_l, loc=<0.0041 0 0>, rot=<0 0 0 1>}
 deform_upperarm_bk_base_l {parent=upperarm_twist_01_l, loc=<1.0813 -7.2193 -1.9453>, rot=<0 0 0 1>}
 deform_upperarm_bk_l {parent=deform_upperarm_bk_base_l, loc=<0 0 0>, rot=<0 0 0 1>}
 deform_upperarm_fr_base_l {parent=upperarm_twist_01_l, loc=<-0.3691 7.0506 -2.765>, rot=<0 0 0 1>}
 deform_upperarm_fr_l {parent=deform_upperarm_fr_base_l, loc=<0 0 0>, rot=<0 0 0 1>}
 deform_upperarm_up_base_l {parent=upperarm_twist_01_l, loc=<-3.2216 -1.1391 3.3311>, rot=<0 0 0 1>}
 deform_upperarm_up_l {parent=deform_upperarm_up_base_l, loc=<0 0 0>, rot=<0 0 0 1>}
 upperarm_twist_02_l {parent=upperarm_l, loc=<13.7941 0 0>, rot=<0 0 0 1>}
 clavicle_r {parent=spine_05, loc=<4.3575 -0.1466 2.3223>, rot=<0.7289 0.0465 -0.6829 0.009>}
 upperarm_r {parent=clavicle_r, loc=<-15.5111 0 0>, rot=<0.0226 -0.3234 -0.0088 -0.9459>}
 lowerarm_r {parent=upperarm_r, loc=<-26.6514 0 0>, rot=<0.0391 -0.0099 -0.2453 -0.9686>}
 hand_r {parent=lowerarm_r, loc=<-22.3491 0 0>, rot=<-0.7081 -0.0507 0.0975 -0.6975>}
 thumb_01_r {parent=hand_r, loc=<-1.3118 0.3191 2.459>, rot=<0.6546 -0.4042 0.0664 -0.6354>}
 thumb_02_r {parent=thumb_01_r, loc=<-5.9336 0.0448 -0.0401>, rot=<-0.0037 0.0213 0.298 -0.9543>}
 thumb_03_r {parent=thumb_02_r, loc=<-2.2223 -0.3682 0.0907>, rot=<0.0118 -0.013 0.0066 -0.9998>}
 index_metacarpal_r {parent=hand_r, loc=<-1.6535 -0.0183 2.96>, rot=<0.0158 -0.0124 -0.0173 -0.9996>}
 index_01_r {parent=index_metacarpal_r, loc=<-8.3008 -0.0357 0.0524>, rot=<-0.0044 0.0485 0.0924 -0.9945>}
 index_02_r {parent=index_01_r, loc=<-3.3587 0.3004 -0.0452>, rot=<-0.002 0.0027 0.3225 -0.9466>}
 index_03_r {parent=index_02_r, loc=<-2.5578 0.123 0.0252>, rot=<-0.002 0.0067 0.2558 -0.9667>}
 middle_metacarpal_r {parent=hand_r, loc=<-2.1856 0.1725 0.5984>, rot=<0.006 0.0168 -0.0329 -0.9993>}
 middle_01_r {parent=middle_metacarpal_r, loc=<-7.3275 0.2447 -0.0129>, rot=<-0.0208 0.0042 0.0858 -0.9961>}
 middle_02_r {parent=middle_01_r, loc=<-4.0123 -0.1401 -0.1175>, rot=<-0.0218 0.0258 0.3512 -0.9357>}
 middle_03_r {parent=middle_02_r, loc=<-2.6077 -0.0534 -0.1373>, rot=<0.0082 0.0188 0.1708 -0.9851>}
 ring_metacarpal_r {parent=hand_r, loc=<-2.2074 0.3625 -1.3638>, rot=<-0.0284 0.0722 -0.0458 -0.9959>}
 ring_01_r {parent=ring_metacarpal_r, loc=<-7.0295 0.6197 0.1446>, rot=<-0.0489 -0.0252 0.1892 -0.9804>}
 ring_02_r {parent=ring_01_r, loc=<-3.6174 -0.4647 0.1027>, rot=<0 0 0.2806 -0.9598>}
 ring_03_r {parent=ring_02_r, loc=<-2.4152 -0.114 -0.0199>, rot=<0 0 0.1189 -0.9929>}
 pinky_metacarpal_r {parent=hand_r, loc=<-1.7616 0.9753 -3.3814>, rot=<-0.0397 0.0814 -0.069 -0.9935>}
 pinky_01_r {parent=pinky_metacarpal_r, loc=<-7.1709 0.4831 0.1094>, rot=<-0.0654 -0.0035 0.2242 -0.9723>}
 pinky_02_r {parent=pinky_01_r, loc=<-2.8274 -0.0893 0.1043>, rot=<0.0037 0.0011 0.2854 -0.9584>}
 pinky_03_r {parent=pinky_02_r, loc=<-2.2046 -0.0307 0.0078>, rot=<-0.0002 0.0098 0.1445 -0.9895>}
 deform_wrist_dn_base_r {parent=hand_r, loc=<-0.0004 2.2345 -0.0001>, rot=<0 0 0 -1>}
 deform_wrist_dn_r {parent=deform_wrist_dn_base_r, loc=<0 0 0>, rot=<0 0 0 1>}
 deform_wrist_up_base_r {parent=hand_r, loc=<0.0003 -2.2684 0.0001>, rot=<0 0 0 -1>}
 deform_wrist_up_r {parent=deform_wrist_up_base_r, loc=<0 0 0>, rot=<0 0 0 1>}
 lowerarm_twist_02_r {parent=lowerarm_r, loc=<-9.6199 0 0>, rot=<0 0 0 -1>}
 lowerarm_twist_01_r {parent=lowerarm_r, loc=<-17.8549 0 0>, rot=<0 0 0 -1>}
 deform_elbow_in_base_r {parent=lowerarm_r, loc=<-0.2111 -5.2591 -0.8275>, rot=<0 0 0 -1>}
 deform_elbow_in_r {parent=deform_elbow_in_base_r, loc=<-0.0001 0 0.0002>, rot=<0 0 0 1>}
 deform_elbow_out_base_r {parent=lowerarm_r, loc=<0 4 -0.0001>, rot=<0 0 0 -1>}
 deform_elbow_out_r {parent=deform_elbow_out_base_r, loc=<0 0 0>, rot=<0 0 0 1>}
 upperarm_twist_01_r {parent=upperarm_r, loc=<-0.0041 0 0>, rot=<0 0 0 -1>}
 deform_upperarm_bk_base_r {parent=upperarm_twist_01_r, loc=<-1.0814 7.2192 1.9455>, rot=<0 0 0 -1>}
 deform_upperarm_bk_r {parent=deform_upperarm_bk_base_r, loc=<0 0 0>, rot=<0 0 0 1>}
 deform_upperarm_fr_base_r {parent=upperarm_twist_01_r, loc=<0.3689 -7.0506 2.7652>, rot=<0 0 0 -1>}
 deform_upperarm_fr_r {parent=deform_upperarm_fr_base_r, loc=<0 0 0>, rot=<0 0 0 1>}
 deform_upperarm_up_base_r {parent=upperarm_twist_01_r, loc=<3.2214 1.1391 -3.3308>, rot=<0 0 0 -1>}
 deform_upperarm_up_r {parent=deform_upperarm_up_base_r, loc=<0 0 0>, rot=<0 0 0 -1>}
 upperarm_twist_02_r {parent=upperarm_r, loc=<-13.7941 0 0>, rot=<0 0 0 -1>}
 collar_ft_l_01_jnt {parent=spine_05, loc=<10.1158 6.1562 -5.4941>, rot=<-0.1263 -0.1239 -0.1197 -0.9769>}
 collar_ft_l_02_jnt {parent=collar_ft_l_01_jnt, loc=<6.4148 0 0>, rot=<0 0 0 1>}
 collar_bk_l_01_jnt {parent=spine_05, loc=<14.8893 0.354 -5.9355>, rot=<0.1486 -0.026 -0.0903 -0.9844>}
 collar_bk_l_02_jnt {parent=collar_bk_l_01_jnt, loc=<5.3233 0 0>, rot=<0 0 0 1>}
 collar_bk_c_01_jnt {parent=spine_05, loc=<16.9772 -3.5093 0>, rot=<0 0 -0.0089 -1>}
 collar_bk_c_02_jnt {parent=collar_bk_c_01_jnt, loc=<5.1773 0 0>, rot=<0 0 0 1>}
 collar_ft_r_01_jnt {parent=spine_05, loc=<10.1154 6.1562 5.4941>, rot=<0.9769 0.1197 0.1239 0.1263>}
 collar_ft_r_02_jnt {parent=collar_ft_r_01_jnt, loc=<6.415 0.0001 0>, rot=<0 0 0 1>}
 collar_bk_r_01_jnt {parent=spine_05, loc=<14.8893 0.354 5.9355>, rot=<0.9844 0.0903 0.026 -0.1486>}
 collar_bk_r_02_jnt {parent=collar_bk_r_01_jnt, loc=<5.3231 -0.0001 0>, rot=<0 0 0 1>}
 deform_pec_aim_base_l {parent=spine_04, loc=<7.8732 12.6388 -2.4123>, rot=<0.0102 -0.4602 0.3979 -0.7936>}
 deform_pec_aim_l {parent=deform_pec_aim_base_l, loc=<0 0 0>, rot=<0 0 0 1>}
 deform_pec_base_l {parent=deform_pec_aim_l, loc=<10.766 0 0>, rot=<0 0 0 1>}
 deform_pec_l {parent=deform_pec_base_l, loc=<0.055 0 0>, rot=<0 0 0 1>}
 deform_pec_aim_base_r {parent=spine_04, loc=<7.8727 12.6388 2.4123>, rot=<-0.4602 -0.0102 0.7936 0.3979>}
 deform_pec_aim_r {parent=deform_pec_aim_base_r, loc=<0 0 0>, rot=<0 0 0 1>}
 deform_pec_base_r {parent=deform_pec_aim_r, loc=<-10.7662 -0.0006 -0.0007>, rot=<0 0 0 1>}
 deform_pec_r {parent=deform_pec_base_r, loc=<0 0 0>, rot=<0 0 0 1>}
 dyn_Jacket_bk_r_MSTR {parent=spine_01, loc=<-4.4084 -10.0553 8.6869>, rot=<0.1528 -0.9705 0.183 0.0367>}
 dyn_Jacket_bk_r_01_jnt {parent=dyn_Jacket_bk_r_MSTR, loc=<0 0 0>, rot=<0 0 0 1>}
 dyn_Jacket_bk_r_02_jnt {parent=dyn_Jacket_bk_r_01_jnt, loc=<15.3097 0 0>, rot=<0 0 0 1>}
 dyn_Jacket_bk_r_03_jnt {parent=dyn_Jacket_bk_r_02_jnt, loc=<15.3096 0 0>, rot=<0 0 0 1>}
 dyn_Jacket_bk_r_04_jnt {parent=dyn_Jacket_bk_r_03_jnt, loc=<15.3097 0 0>, rot=<0 0 0 1>}
 dyn_Jacket_bk_r_05_jnt {parent=dyn_Jacket_bk_r_04_jnt, loc=<15.3097 0.0001 0.0001>, rot=<0 0 0 1>}
 dyn_Jacket_r_MSTR {parent=spine_01, loc=<-3.5442 -4.5599 15.8636>, rot=<0.1405 -0.7468 0.6494 0.0284>}
 dyn_Jacket_r_01_jnt {parent=dyn_Jacket_r_MSTR, loc=<0 0 0>, rot=<0 0 0 1>}
 dyn_Jacket_r_02_jnt {parent=dyn_Jacket_r_01_jnt, loc=<15.3097 0.0001 0>, rot=<0 0 0 1>}
 dyn_Jacket_r_03_jnt {parent=dyn_Jacket_r_02_jnt, loc=<15.3097 0 0>, rot=<0 0 0 1>}
 dyn_Jacket_r_04_jnt {parent=dyn_Jacket_r_03_jnt, loc=<15.3097 0.0001 0>, rot=<0 0 0 1>}
 dyn_Jacket_r_05_jnt {parent=dyn_Jacket_r_04_jnt, loc=<15.3097 0 0>, rot=<0 0 0 1>}
 dyn_Jacket_bk_l_MSTR {parent=spine_01, loc=<-4.4084 -10.0552 -8.6869>, rot=<0.1528 -0.9705 -0.183 -0.0367>}
 dyn_Jacket_bk_l_01_jnt {parent=dyn_Jacket_bk_l_MSTR, loc=<0 0 0>, rot=<0 0 0 1>}
 dyn_Jacket_bk_l_02_jnt {parent=dyn_Jacket_bk_l_01_jnt, loc=<15.3097 0 0>, rot=<0 0 0 1>}
 dyn_Jacket_bk_l_03_jnt {parent=dyn_Jacket_bk_l_02_jnt, loc=<15.3097 0 0>, rot=<0 0 0 1>}
 dyn_Jacket_bk_l_04_jnt {parent=dyn_Jacket_bk_l_03_jnt, loc=<15.3097 0 0>, rot=<0 0 0 1>}
 dyn_Jacket_bk_l_05_jnt {parent=dyn_Jacket_bk_l_04_jnt, loc=<15.3097 0 0>, rot=<0 0 0 1>}
 dyn_Jacket_l_MSTR {parent=spine_01, loc=<-3.5443 -4.5599 -15.8636>, rot=<0.1405 -0.7468 -0.6494 -0.0284>}
 dyn_Jacket_l_01_jnt {parent=dyn_Jacket_l_MSTR, loc=<0 0 0>, rot=<0 0 0 1>}
 dyn_Jacket_l_02_jnt {parent=dyn_Jacket_l_01_jnt, loc=<15.3097 0 0>, rot=<0 0 0 1>}
 dyn_Jacket_l_03_jnt {parent=dyn_Jacket_l_02_jnt, loc=<15.3097 0 0>, rot=<0 0 0 1>}
 dyn_Jacket_l_04_jnt {parent=dyn_Jacket_l_03_jnt, loc=<15.3097 0 0>, rot=<0 0 0 1>}
 dyn_Jacket_l_05_jnt {parent=dyn_Jacket_l_04_jnt, loc=<15.3097 0 0>, rot=<0 0 0 1>}
 dyn_Jacket_bk_c_MSTR {parent=spine_01, loc=<-3.8456 -11.8504 0>, rot=<0.1579 -0.9875 0 0>}
 dyn_Jacket_bk_c_01_jnt {parent=dyn_Jacket_bk_c_MSTR, loc=<0 0 0>, rot=<0 0 0 1>}
 dyn_Jacket_bk_c_02_jnt {parent=dyn_Jacket_bk_c_01_jnt, loc=<15.3097 0 0>, rot=<0 0 0 1>}
 dyn_Jacket_bk_c_03_jnt {parent=dyn_Jacket_bk_c_02_jnt, loc=<15.3097 0 0>, rot=<0 0 0 1>}
 dyn_Jacket_bk_c_04_jnt {parent=dyn_Jacket_bk_c_03_jnt, loc=<15.3097 0 0>, rot=<0 0 0 1>}
 dyn_Jacket_bk_c_05_jnt {parent=dyn_Jacket_bk_c_04_jnt, loc=<15.3097 0 0>, rot=<0 0 0 1>}
 thigh_l {parent=pelvis, loc=<-2.5693 -0.0109 -7.7501>, rot=<0.0012 0.0572 0.0155 -0.9982>}
 calf_l {parent=thigh_l, loc=<-39.0484 0 0>, rot=<0.0006 -0.0161 -0.0391 -0.9991>}
 foot_l {parent=calf_l, loc=<-42.561 0.1077 0.4292>, rot=<0.0418 -0.0455 0.0343 -0.9975>}
 ball_l {parent=foot_l, loc=<-6.1599 13.8264 0.6806>, rot=<0.0001 -0.0001 -0.7063 -0.7079>}
 calf_twist_02_l {parent=calf_l, loc=<-18.8064 -0.0967 0.0397>, rot=<0.0237 0.0095 -0.0035 -0.9997>}
 calf_twist_01_l {parent=calf_l, loc=<-35.813 -0.0967 0.0397>, rot=<0.0237 0.0095 -0.0035 -0.9997>}
 deform_knee_in_base_l {parent=calf_l, loc=<1.298 -5.2957 -0.0496>, rot=<0 0 0 -1>}
 deform_knee_in_l {parent=deform_knee_in_base_l, loc=<0 0 0>, rot=<0 0 0 1>}
 deform_knee_out_base_l {parent=calf_l, loc=<0 4 0>, rot=<0 0 0 -1>}
 deform_knee_out_l {parent=deform_knee_out_base_l, loc=<0 0 0>, rot=<0 0 0 1>}
 thigh_twist_01_l {parent=thigh_l, loc=<-21.6303 0 0>, rot=<0.0013 0 0 -1>}
 deform_groin_base_l {parent=thigh_l, loc=<-2.1908 8.1142 -1.7442>, rot=<0 0 0 -1>}
 deform_groin_l {parent=deform_groin_base_l, loc=<0 0 0>, rot=<0 0 0 1>}
 thigh_r {parent=pelvis, loc=<-2.5693 -0.0109 7.7501>, rot=<0.0572 -0.0012 0.9982 0.0155>}
 calf_r {parent=thigh_r, loc=<39.0484 0 0>, rot=<0.0006 -0.0161 -0.0391 -0.9991>}
 foot_r {parent=calf_r, loc=<42.561 -0.1077 -0.4292>, rot=<0.0418 -0.0455 0.0343 -0.9975>}
 ball_r {parent=foot_r, loc=<6.1599 -13.8264 -0.6806>, rot=<0.0001 -0.0001 -0.7063 -0.7079>}
 calf_twist_02_r {parent=calf_r, loc=<18.8064 0.0967 -0.0397>, rot=<0.0237 0.0095 -0.0035 -0.9997>}
 calf_twist_01_r {parent=calf_r, loc=<35.813 0.0967 -0.0397>, rot=<0.0237 0.0095 -0.0035 -0.9997>}
 deform_knee_in_base_r {parent=calf_r, loc=<-1.298 5.2957 0.0496>, rot=<0 0 0 -1>}
 deform_knee_in_r {parent=deform_knee_in_base_r, loc=<0 0 0>, rot=<0 0 0 1>}
 deform_knee_out_base_r {parent=calf_r, loc=<0 -4 0>, rot=<0 0 0 -1>}
 deform_knee_out_r {parent=deform_knee_out_base_r, loc=<0 0 0>, rot=<0 0 0 1>}
 thigh_twist_01_r {parent=thigh_r, loc=<21.6303 0 0>, rot=<0.0013 0 0 -1>}
 deform_groin_base_r {parent=thigh_r, loc=<2.1908 -8.1142 1.7442>, rot=<0 0 0 -1>}
 deform_groin_r {parent=deform_groin_base_r, loc=<0 0 0>, rot=<0 0 0 1>}
 deform_glute_base_l {parent=pelvis, loc=<-8.8064 -8.8222 -8.4287>, rot=<0.0275 0.0501 -0.4472 -0.8926>}
 deform_glute_l {parent=deform_glute_base_l, loc=<0 0 0>, rot=<0 0 0 1>}
 deform_glute_base_r {parent=pelvis, loc=<-8.8062 -8.8222 8.4287>, rot=<0.0501 -0.0275 0.8926 -0.4472>}
 deform_glute_r {parent=deform_glute_base_r, loc=<0 0 0>, rot=<0 0 0 1>}
]
variables [
 L calfBend Alpha {type=double, default=0}
 L thighBack Alpha {type=double, default=0}
 L thighForward Alpha {type=double, default=0}
 L thighSide Alpha {type=double, default=0}
 L thighVeryForward {type=double, default=0}
 Movement Alpha {type=double, default=0}
 R calfBend Alpha {type=double, default=0}
 R thighBack Alpha {type=double, default=0}
 R thighForward Alpha {type=double, default=0}
 R thighSide Alpha {type=double, default=0}
 R thighVeryForward {type=double, default=0}
 is_DBNO_Carried {type=bool, default=False}
 pelvisLowToGround_INVERSE {type=double, default=0}
 pelvisLowToGround_translateRotate {type=double, default=0}
]
