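# Physics asset for the M_MED_EarthLane male medium body cosmetic, set up for the dynamic jacket bones over a simplified lower body.
# Five jacket chains (back center, left, right, plus left and right) each run five capsule joints with narrow swing and almost no twist.
# Each chain hangs from a zero-size master body with free rotation on spine_01.
# Capsules cover spine, pelvis and legs, and an oversized root capsule sits low, apparently as a ground collider.
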PhysicsAsset /BRCosmetics/Characters/Player/Male/Medium/Bodies/M_MED_EarthLane/Meshes/M_MED_EarthLane_Physics
bodies {
 calf_l [{shapeType=Capsule, length=30, radius=6.5, center=<-22 0 1>, rotation=<rot 0 90 88>}]
 calf_r [{shapeType=Capsule, length=30, radius=6.5, center=<22 0 -1>, rotation=<rot 0 90 88>}]
 dyn_Jacket_bk_c_01_jnt [{shapeType=Capsule, length=12, radius=8, shapeName=dyn_Jacket_bk_c_01_jnt_capsule, center=<8 -3 0>}]
 dyn_Jacket_bk_c_02_jnt [{shapeType=Capsule, length=12, radius=8, shapeName=dyn_Jacket_bk_c_02_jnt_capsule, center=<8 -3 0>}]
 dyn_Jacket_bk_c_03_jnt [{shapeType=Capsule, length=12, radius=8, shapeName=dyn_Jacket_bk_c_03_jnt_capsule, center=<8 -3 0>}]
 dyn_Jacket_bk_c_04_jnt [{shapeType=Capsule, length=12, radius=8, shapeName=dyn_Jacket_bk_c_04_jnt_capsule, center=<8 -3 0>}]
 dyn_Jacket_bk_c_05_jnt [{shapeType=Capsule, length=12, radius=8, shapeName=dyn_Jacket_bk_c_05_jnt_capsule, center=<8 -3 0>}]
 dyn_Jacket_bk_c_MSTR [{shapeType=Capsule, length=0, radius=0, shapeName=dyn_Jacket_bk_c_MSTR_capsule}]
 dyn_Jacket_bk_l_01_jnt [{shapeType=Capsule, length=12, radius=8, shapeName=dyn_Jacket_bk_l_01_jnt_capsule, center=<8 -3 0>}]
 dyn_Jacket_bk_l_02_jnt [{shapeType=Capsule, length=12, radius=8, center=<8 -3 0>}]
 dyn_Jacket_bk_l_03_jnt [{shapeType=Capsule, length=12, radius=8, shapeName=dyn_Jacket_bk_l_03_jnt_capsule, center=<8 -3 0>}]
 dyn_Jacket_bk_l_04_jnt [{shapeType=Capsule, length=12, radius=8, center=<8 -3 0>}]
 dyn_Jacket_bk_l_05_jnt [{shapeType=Capsule, length=12, radius=8, shapeName=dyn_Jacket_bk_l_05_jnt_capsule, center=<8 -3 0>}]
 dyn_Jacket_bk_l_MSTR [{shapeType=Capsule, length=0, radius=0, shapeName=dyn_Jacket_bk_l_MSTR_capsule}]
 dyn_Jacket_bk_r_01_jnt [{shapeType=Capsule, length=12, radius=8, shapeName=dyn_Jacket_bk_r_01_jnt_capsule, center=<8 -3 0>}]
 dyn_Jacket_bk_r_02_jnt [{shapeType=Capsule, length=12, radius=8, center=<8 -3 0>}]
 dyn_Jacket_bk_r_03_jnt [{shapeType=Capsule, length=12, radius=8, shapeName=dyn_Jacket_bk_r_03_jnt_capsule, center=<8 -3 0>}]
 dyn_Jacket_bk_r_04_jnt [{shapeType=Capsule, length=12, radius=8, center=<8 -3 0>}]
 dyn_Jacket_bk_r_05_jnt [{shapeType=Capsule, length=12, radius=8, shapeName=dyn_Jacket_bk_r_05_jnt_capsule, center=<8 -3 0>}]
 dyn_Jacket_bk_r_MSTR [{shapeType=Capsule, length=0, radius=0, shapeName=dyn_Jacket_bk_r_MSTR_capsule}]
 dyn_Jacket_l_01_jnt [{shapeType=Capsule, length=12, radius=8, shapeName=dyn_Jacket_l_01_jnt_capsule, center=<8 -3 -3>}]
 dyn_Jacket_l_02_jnt [{shapeType=Capsule, length=12, radius=8, shapeName=dyn_Jacket_l_02_jnt_capsule, center=<8 -3 -3>}]
 dyn_Jacket_l_03_jnt [{shapeType=Capsule, length=12, radius=8, shapeName=dyn_Jacket_l_03_jnt_capsule, center=<8 -3 -3>}]
 dyn_Jacket_l_04_jnt [{shapeType=Capsule, length=12, radius=8, center=<8 -3 -3>}]
 dyn_Jacket_l_05_jnt [{shapeType=Capsule, length=12, radius=8, shapeName=dyn_Jacket_l_05_jnt_capsule, center=<8 -3 -3>}]
 dyn_Jacket_l_MSTR [{shapeType=Capsule, length=0, radius=0, shapeName=dyn_Jacket_l_MSTR_capsule}]
 dyn_Jacket_r_01_jnt [{shapeType=Capsule, length=12, radius=8, center=<8 -3 3>}]
 dyn_Jacket_r_02_jnt [{shapeType=Capsule, length=12, radius=8, shapeName=dyn_Jacket_r_02_jnt_capsule, center=<8 -3 3>}]
 dyn_Jacket_r_03_jnt [{shapeType=Capsule, length=12, radius=8, shapeName=dyn_Jacket_r_03_jnt_capsule, center=<8 -3 3>}]
 dyn_Jacket_r_04_jnt [{shapeType=Capsule, length=12, radius=8, center=<8 -3 3>}]
 dyn_Jacket_r_05_jnt [{shapeType=Capsule, length=12, radius=8, shapeName=dyn_Jacket_r_05_jnt_capsule, center=<8 -3 3>}]
 dyn_Jacket_r_MSTR [{shapeType=Capsule, length=0, radius=0, shapeName=dyn_Jacket_r_MSTR_capsule}]
 foot_l [{shapeType=Capsule, length=15, radius=6.4, center=<-4.3 6 0>, rotation=<rot 0 0 90>}]
 foot_r [{shapeType=Capsule, length=15, radius=6.4, center=<4.3 -6 0>, rotation=<rot 0 0 90>}]
 pelvis [{shapeType=Capsule, length=3, radius=12}]
 root [{shapeType=Capsule, length=60, radius=200, shapeName=root_capsule, center=<0 0 -185>, rotation=<rot 0 90 90>}]
 spine_01 [{shapeType=Capsule, length=3, radius=12}]
 thigh_l [{shapeType=Capsule, length=30, radius=7.5, center=<-20 0 1>, rotation=<rot 0 90 88>}]
 thigh_r [{shapeType=Capsule, length=30, radius=7.5, center=<20 0 -1>, rotation=<rot 0 90 88>}]
}
constraints [
 {bone1Name=spine_01, bone2Name=pelvis, swing1LimitDegrees=45, swing2LimitDegrees=45, twistLimitDegrees=45}
 {bone1Name=dyn_Jacket_bk_l_02_jnt, bone2Name=dyn_Jacket_bk_l_01_jnt, swing1LimitDegrees=15, swing2LimitDegrees=15, twistLimitDegrees=1}
 {bone1Name=dyn_Jacket_bk_l_04_jnt, bone2Name=dyn_Jacket_bk_l_03_jnt, swing1LimitDegrees=15, swing2LimitDegrees=15, twistLimitDegrees=1}
 {bone1Name=dyn_Jacket_l_04_jnt, bone2Name=dyn_Jacket_l_03_jnt, swing1LimitDegrees=15, swing2LimitDegrees=15, twistLimitDegrees=1}
 {bone1Name=dyn_Jacket_bk_r_02_jnt, bone2Name=dyn_Jacket_bk_r_01_jnt, swing1LimitDegrees=15, swing2LimitDegrees=15, twistLimitDegrees=1}
 {bone1Name=dyn_Jacket_bk_r_04_jnt, bone2Name=dyn_Jacket_bk_r_03_jnt, swing1LimitDegrees=15, swing2LimitDegrees=15, twistLimitDegrees=1}
 {bone1Name=dyn_Jacket_r_01_jnt, bone2Name=dyn_Jacket_r_MSTR, swing1LimitDegrees=15, swing2LimitDegrees=15, twistLimitDegrees=1}
 {bone1Name=dyn_Jacket_r_04_jnt, bone2Name=dyn_Jacket_r_03_jnt, swing1LimitDegrees=15, swing2LimitDegrees=15, twistLimitDegrees=1}
 {bone1Name=thigh_l, bone2Name=pelvis, swing1LimitDegrees=45, swing2LimitDegrees=45, twistLimitDegrees=45}
 {bone1Name=calf_l, bone2Name=thigh_l, swing1LimitDegrees=45, swing2LimitDegrees=45, twistLimitDegrees=45}
 {bone1Name=foot_l, bone2Name=calf_l, swing1LimitDegrees=45, swing2LimitDegrees=45, twistLimitDegrees=45}
 {bone1Name=thigh_r, bone2Name=pelvis, swing1LimitDegrees=45, swing2LimitDegrees=45, twistLimitDegrees=45}
 {bone1Name=calf_r, bone2Name=thigh_r, swing1LimitDegrees=45, swing2LimitDegrees=45, twistLimitDegrees=45}
 {bone1Name=foot_r, bone2Name=calf_r, swing1LimitDegrees=45, swing2LimitDegrees=45, twistLimitDegrees=45}
 {
  bone1Name=dyn_Jacket_bk_c_MSTR
  bone2Name=spine_01
  swing1LimitDegrees=45
  swing1Motion=Free
  swing2LimitDegrees=45
  swing2Motion=Free
  twistLimitDegrees=45
  twistMotion=Free
 }
 {bone1Name=dyn_Jacket_bk_c_01_jnt, bone2Name=dyn_Jacket_bk_c_MSTR, swing1LimitDegrees=15, swing2LimitDegrees=15, twistLimitDegrees=1}
 {bone1Name=dyn_Jacket_bk_c_02_jnt, bone2Name=dyn_Jacket_bk_c_01_jnt, swing1LimitDegrees=15, swing2LimitDegrees=15, twistLimitDegrees=1}
 {bone1Name=dyn_Jacket_bk_c_03_jnt, bone2Name=dyn_Jacket_bk_c_02_jnt, swing1LimitDegrees=15, swing2LimitDegrees=15, twistLimitDegrees=1}
 {bone1Name=dyn_Jacket_bk_c_04_jnt, bone2Name=dyn_Jacket_bk_c_03_jnt, swing1LimitDegrees=15, swing2LimitDegrees=15, twistLimitDegrees=1}
 {bone1Name=dyn_Jacket_bk_c_05_jnt, bone2Name=dyn_Jacket_bk_c_04_jnt, swing1LimitDegrees=15, swing2LimitDegrees=15, twistLimitDegrees=1}
 {
  bone1Name=dyn_Jacket_bk_l_MSTR
  bone2Name=spine_01
  swing1LimitDegrees=45
  swing1Motion=Free
  swing2LimitDegrees=45
  swing2Motion=Free
  twistLimitDegrees=45
  twistMotion=Free
 }
 {bone1Name=dyn_Jacket_bk_l_01_jnt, bone2Name=dyn_Jacket_bk_l_MSTR, swing1LimitDegrees=15, swing2LimitDegrees=15, twistLimitDegrees=1}
 {bone1Name=dyn_Jacket_bk_l_03_jnt, bone2Name=dyn_Jacket_bk_l_02_jnt, swing1LimitDegrees=15, swing2LimitDegrees=15, twistLimitDegrees=1}
 {bone1Name=dyn_Jacket_bk_l_05_jnt, bone2Name=dyn_Jacket_bk_l_04_jnt, swing1LimitDegrees=15, swing2LimitDegrees=15, twistLimitDegrees=1}
 {
  bone1Name=dyn_Jacket_bk_r_MSTR
  bone2Name=spine_01
  swing1LimitDegrees=45
  swing1Motion=Free
  swing2LimitDegrees=45
  swing2Motion=Free
  twistLimitDegrees=45
  twistMotion=Free
 }
 {bone1Name=dyn_Jacket_bk_r_01_jnt, bone2Name=dyn_Jacket_bk_r_MSTR, swing1LimitDegrees=15, swing2LimitDegrees=15, twistLimitDegrees=1}
 {bone1Name=dyn_Jacket_bk_r_03_jnt, bone2Name=dyn_Jacket_bk_r_02_jnt, swing1LimitDegrees=15, swing2LimitDegrees=15, twistLimitDegrees=1}
 {bone1Name=dyn_Jacket_bk_r_05_jnt, bone2Name=dyn_Jacket_bk_r_04_jnt, swing1LimitDegrees=15, swing2LimitDegrees=15, twistLimitDegrees=1}
 {
  bone1Name=dyn_Jacket_l_MSTR
  bone2Name=spine_01
  swing1LimitDegrees=45
  swing1Motion=Free
  swing2LimitDegrees=45
  swing2Motion=Free
  twistLimitDegrees=45
  twistMotion=Free
 }
 {bone1Name=dyn_Jacket_l_01_jnt, bone2Name=dyn_Jacket_l_MSTR, swing1LimitDegrees=15, swing2LimitDegrees=15, twistLimitDegrees=1}
 {bone1Name=dyn_Jacket_l_02_jnt, bone2Name=dyn_Jacket_l_01_jnt, swing1LimitDegrees=15, swing2LimitDegrees=15, twistLimitDegrees=1}
 {bone1Name=dyn_Jacket_l_03_jnt, bone2Name=dyn_Jacket_l_02_jnt, swing1LimitDegrees=15, swing2LimitDegrees=15, twistLimitDegrees=1}
 {bone1Name=dyn_Jacket_l_05_jnt, bone2Name=dyn_Jacket_l_04_jnt, swing1LimitDegrees=15, swing2LimitDegrees=15, twistLimitDegrees=1}
 {
  bone1Name=dyn_Jacket_r_MSTR
  bone2Name=spine_01
  swing1LimitDegrees=45
  swing1Motion=Free
  swing2LimitDegrees=45
  swing2Motion=Free
  twistLimitDegrees=45
  twistMotion=Free
 }
 {bone1Name=dyn_Jacket_r_02_jnt, bone2Name=dyn_Jacket_r_01_jnt, swing1LimitDegrees=15, swing2LimitDegrees=15, twistLimitDegrees=1}
 {bone1Name=dyn_Jacket_r_03_jnt, bone2Name=dyn_Jacket_r_02_jnt, swing1LimitDegrees=15, swing2LimitDegrees=15, twistLimitDegrees=1}
 {bone1Name=dyn_Jacket_r_05_jnt, bone2Name=dyn_Jacket_r_04_jnt, swing1LimitDegrees=15, swing2LimitDegrees=15, twistLimitDegrees=1}
 {
  bone1Name=pelvis
  bone2Name=root
  swing1LimitDegrees=45
  swing1Motion=Free
  swing2LimitDegrees=45
  swing2Motion=Free
  twistLimitDegrees=45
  twistMotion=Free
 }
]
