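# Animation Blueprint for the male EarthLane face accessory mesh, which follows the character's pose and carries dynamic hair joints.
# It copies the attached parent's pose with curves, layering in the cached body mesh pose from both upper arms down.
# A chain of bone-space Replace nodes then sets nine hair root joints, with zero rotations shown.
# Each update computes three time-driven random shake rotators, not visibly wired to those joints, and re-caches the body mesh.
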AnimBlueprint /BRCosmetics/Characters/Player/Male/Medium/Bodies/M_MED_EarthLane/Meshes/Parts/M_MED_EarthLane_FaceAcc_AnimBP
target_skeleton=&/M_MED_EarthLane_FaceAcc_Skeleton
graphs [
 AnimGraph {nodes [
   n1 = Copy Pose From Mesh() {bCopyCurves=true, bUseAttachedParent=true}
   n2 = Copy Pose From Mesh(SourceMeshComponent=Get body Mesh() -> (body Mesh:Skeletal Mesh Component Object Reference))
   n3 = "Transform (Modify) Bone - Bone: dyn_Hair_tp_c_01_jnt"(
    ComponentPose=Local To Component(LocalPose=Layered blend per bone(BasePose=@n1, BlendPoses_0=@n2) {layerSetup [{branchFilters [{boneName=upperarm_l}, {boneName=upperarm_r}]}]})
    Alpha=1
    Translation="0,0,0"
    Rotation="0,0,0"
    Scale="1,1,1"
   ) {rotationMode=BMM_Replace, rotationSpace=BCS_BoneSpace, alphaBoolBlend {blendOption=Linear}, boneToModify {boneName=dyn_Hair_tp_c_01_jnt}}
   n4 = "Transform (Modify) Bone - Bone: dyn_Hair_ft_l_01_jnt"(ComponentPose=@n3, Alpha=1, Translation="0,0,0", Rotation="0,0,0", Scale="1,1,1") {rotationMode=BMM_Replace, rotationSpace=BCS_BoneSpace, alphaBoolBlend {blendOption=Linear}, boneToModify {boneName=dyn_Hair_ft_l_01_jnt}}
   n5 = "Transform (Modify) Bone - Bone: dyn_Hair_bk_r_01_jnt"(ComponentPose=@n4, Alpha=1, Translation="0,0,0", Rotation="0,0,0", Scale="1,1,1") {rotationMode=BMM_Replace, rotationSpace=BCS_BoneSpace, alphaBoolBlend {blendOption=Linear}, boneToModify {boneName=dyn_Hair_bk_r_01_jnt}}
   n6 = "Transform (Modify) Bone - Bone: dyn_Hair_ft_c_01_jnt"(ComponentPose=@n5, Alpha=1, Translation="0,0,0", Rotation="0,0,0", Scale="1,1,1") {rotationMode=BMM_Replace, rotationSpace=BCS_BoneSpace, alphaBoolBlend {blendOption=Linear}, boneToModify {boneName=dyn_Hair_ft_c_01_jnt}}
   n7 = "Transform (Modify) Bone - Bone: dyn_Hair_bk_l_01_jnt"(ComponentPose=@n6, Alpha=1, Translation="0,0,0", Rotation="0,0,0", Scale="1,1,1") {rotationMode=BMM_Replace, rotationSpace=BCS_BoneSpace, alphaBoolBlend {blendOption=Linear}, boneToModify {boneName=dyn_Hair_bk_l_01_jnt}}
   n8 = "Transform (Modify) Bone - Bone: dyn_Hair_tp_r_01_jnt"(ComponentPose=@n7, Alpha=1, Translation="0,0,0", Rotation="0,0,0", Scale="1,1,1") {rotationMode=BMM_Replace, rotationSpace=BCS_BoneSpace, alphaBoolBlend {blendOption=Linear}, boneToModify {boneName=dyn_Hair_tp_r_01_jnt}}
   n9 = "Transform (Modify) Bone - Bone: dyn_Hair_bk_c_01_jnt"(ComponentPose=@n8, Alpha=1, Translation="0,0,0", Rotation="0,0,0", Scale="1,1,1") {rotationMode=BMM_Replace, rotationSpace=BCS_BoneSpace, alphaBoolBlend {blendOption=Linear}, boneToModify {boneName=dyn_Hair_bk_c_01_jnt}}
   n10 = "Transform (Modify) Bone - Bone: dyn_Hair_tp_l_01_jnt"(ComponentPose=@n9, Alpha=1, Translation="0,0,0", Rotation="0,0,0", Scale="1,1,1") {rotationMode=BMM_Replace, rotationSpace=BCS_BoneSpace, alphaBoolBlend {blendOption=Linear}, boneToModify {boneName=dyn_Hair_tp_l_01_jnt}}
   n11 = "Transform (Modify) Bone - Bone: dyn_Hair_ft_r_01_jnt"(ComponentPose=@n10, Alpha=1, Translation="0,0,0", Rotation="0,0,0", Scale="1,1,1") {rotationMode=BMM_Replace, rotationSpace=BCS_BoneSpace, alphaBoolBlend {blendOption=Linear}, boneToModify {boneName=dyn_Hair_ft_r_01_jnt}}
   Output Pose(Result=Component To Local(ComponentPose=@n11)) {layerGroup=None}
  ]}
 EventGraph {nodes [
   Event BlueprintUpdateAnimation()
   > n1 = Sequence()
   n2 = float * float(A=GetGameTimeInSeconds(), B=Get Shake Frequency() -> (Shake Frequency:double)) {is=title K2Node_PromotableOperator}
   > n3 = RandomRotatorFunction!(
    execute
    GameTimeInSeconds:double=@n2
    PerlinScale:double=1
    Frequency:Rotator="7.5,7.5,0"
    Amplitude:Rotator="3.5,3.5,0"
    Offset:Rotator="0, 0, 0"
    TimeOffset:double=1
   ) -> (then, RandomRotator:Rotator)
   > Set HairRotator_01(HairRotator_01:Rotator=@n3) -> (Output_Get:Rotator)
   > n4 = RandomRotatorFunction(GameTimeInSeconds=@n2, PerlinScale=1, Frequency="7,7,0", Amplitude="3.5,3.5,0", Offset="0,3, 0", TimeOffset=2)
   > Set HairRotator_02(HairRotator_02:Rotator=@n4) -> (Output_Get:Rotator)
   > n5 = RandomRotatorFunction(GameTimeInSeconds=@n2, PerlinScale=1, Frequency="8,8,0", Amplitude="3.5,3.5,0", Offset="0, 0, 0", TimeOffset=0)
   > Set HairRotator_03(HairRotator_03:Rotator=@n5) -> (Output_Get:Rotator)
   Set body Mesh(execute=@n1.then_1, body Mesh:Skeletal Mesh Component Object Reference=GetOwnerSkeletalMeshComponent()) -> (Output_Get:Skeletal Mesh Component Object Reference)
   Event BlueprintInitializeAnimation()
   > Set body Mesh(body Mesh=GetOwnerSkeletalMeshComponent())
  ]}
]
variables [
 body Mesh {type=SkeletalMeshComponent}
 HairRotator_01 {type=Rotator}
 HairRotator_02 {type=Rotator}
 HairRotator_03 {type=Rotator}
 Shake Frequency {type=number, default=1}
]
